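# Animation Blueprint for the Skirmish Male pickaxe on the melee weapon skeleton, flipping the weapon when it is the offhand mesh.
# On initialization it fetches its part's skeletal mesh component and sets an offhand flag if that component's string name equals "WeaponMeshOffhand".
# The AnimGraph uses that flag to enable a Modify Bone node that additively yaws b_AR_Root 180 degrees in bone space.
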AnimBlueprint /BRCosmetics/Weapons/FORT_Melee/Pickaxe_Skirmish_Male/Meshes/SkirmishMale_AnimBP
target_skeleton=/Game/Weapons/FORT_Melee/Meshes/SK_Melee_Skeleton
graphs [
 AnimGraph {nodes [
   n1 = "Transform (Modify) Bone - Bone: b_AR_Root"!(
    ComponentPose=Local To Component()
    bAlphaBoolEnabled:bool=Get offhand_weapon() -> (offhand_weapon:bool)
    Translation="0,0,0"
    Rotation="0,180,0"
    Scale="1,1,1"
   ) -> (Pose) {
    alphaInputType=Bool
    rotationMode=BMM_Additive
    rotationSpace=BCS_BoneSpace
    translationMode=BMM_Additive
    translationSpace=BCS_BoneSpace
    alphaBoolBlend {blendOption=Linear}
    boneToModify {boneName=b_AR_Root}
    rotation {yaw=180}
   }
   Output Pose(Result=Component To Local(ComponentPose=@n1)) {layerGroup=None}
  ]}
 EventGraph {nodes [
   Event BlueprintInitializeAnimation()
   > n1 = Set Skel_Mesh(Skel_Mesh=GetPartSkeletalMeshComponent())
   > n2 = Branch(Condition="Equal Exactly (String)"(A="To String (Object)"!(InObj:Object Reference=@n1) -> (ReturnValue:String), B=WeaponMeshOffhand))
   > Set offhand_weapon(offhand_weapon:bool=true) -> (Output_Get:bool)
   Set offhand_weapon(execute=@n2.else, offhand_weapon=false)
  ]}
]
variables [
 Offhand_Weapon {type=boolean}
 Skel_Mesh {type=SkeletalMeshComponent}
]
